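# Physics asset for the dangling accessories on the female medium Fisherman body cosmetic: a bobber, three bullets, a walkie-talkie and a pouch.
# Each accessory has a capsule body constrained to its own tiny _MSTR anchor body, with the bobber allowed the widest swing.
# The pouch continues into a five-link chain of tightly limited joints, where the later links bend mostly on one swing axis.
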PhysicsAsset /BRCosmetics/Characters/Player/Female/Medium/Bodies/F_MED_Fisherman/Meshes/F_MED_Fisherman_Dyn
bodies {
 dyn_bobber [{shapeType=Capsule, length=17.8125, radius=1.0625, center=<8.5415 0 0>, rotation=<rot 90 0 0>}]
 dyn_bobber_MSTR [{shapeType=Capsule, length=0.1, radius=0.1}]
 dyn_bullet_01 [{shapeType=Capsule, length=2.4062, radius=1.5938, center=<0 0 1.2>, rotation=<rot 90 0 0>}]
 dyn_bullet_01_MSTR [{shapeType=Capsule, length=0.1, radius=0.1}]
 dyn_bullet_02 [{shapeType=Capsule, length=2.4062, radius=1.5938, center=<0 0 1.2>, rotation=<rot 90 0 0>}]
 dyn_bullet_02_MSTR [{shapeType=Capsule, length=0.1, radius=0.1}]
 dyn_bullet_03 [{shapeType=Capsule, length=2.4062, radius=1.5938, center=<0 0 1.2>, rotation=<rot 90 0 0>}]
 dyn_bullet_03_MSTR [{shapeType=Capsule, length=0.1, radius=0.1}]
 dyn_pouch [{shapeType=Capsule, length=5.0169, radius=4.8302, center=<2.0407 0.4511 1.1761>, rotation=<rot -90 -90 99.9999>}]
 dyn_pouch1 [{shapeType=Capsule, length=4.75, radius=1.25, center=<1.1404 0 0.5304>, rotation=<rot 90 0 90>}]
 dyn_pouch2 [{shapeType=Capsule, length=4.75, radius=1.25, center=<0.6874 0 -0.2506>, rotation=<rot 90 -11.3099 68.6885>}]
 dyn_pouch3 [{shapeType=Capsule, length=5, radius=1.25, center=<0 0.0001 0.1076>, rotation=<rot 90 0 90>}]
 dyn_pouch4 [{shapeType=Capsule, length=1, radius=1, center=<1.4398 0.2538 -0.4639>, rotation=<rot 79.9996 9.9988 0>}]
 dyn_pouch5 [{shapeType=Capsule, length=1.25, radius=1, center=<1.4993 0 -0.0001>, rotation=<rot 90 -0.0302 -0.0306>}]
 dyn_pouch_MSTR [{shapeType=Capsule, length=0.505, radius=0.505, rotation=<rot 0 -90 0>}]
 dyn_talkie [{shapeType=Capsule, length=9.0625, radius=2, center=<1.26 0 1.75>, rotation=<rot 90 0 0>}]
 dyn_talkie_MSTR [{shapeType=Capsule, length=0.1, radius=0.1}]
}
constraints [
 {bone1Name=dyn_bullet_01, bone2Name=dyn_bullet_01_MSTR, swing1LimitDegrees=45, swing2LimitDegrees=45, twistLimitDegrees=5}
 {bone1Name=dyn_bullet_02, bone2Name=dyn_bullet_02_MSTR, swing1LimitDegrees=45, swing2LimitDegrees=45, twistLimitDegrees=5}
 {bone1Name=dyn_bullet_03, bone2Name=dyn_bullet_03_MSTR, swing1LimitDegrees=45, swing2LimitDegrees=45, twistLimitDegrees=5}
 {bone1Name=dyn_bobber, bone2Name=dyn_bobber_MSTR, swing1LimitDegrees=110, swing2LimitDegrees=60, twistLimitDegrees=5}
 {bone1Name=dyn_talkie, bone2Name=dyn_talkie_MSTR, swing1LimitDegrees=45, swing2LimitDegrees=45, twistLimitDegrees=5}
 {bone1Name=dyn_pouch, bone2Name=dyn_pouch_MSTR, swing1LimitDegrees=5, swing2LimitDegrees=5, twistLimitDegrees=5}
 {bone1Name=dyn_pouch1, bone2Name=dyn_pouch, swing1LimitDegrees=5, swing2LimitDegrees=10, twistLimitDegrees=1}
 {bone1Name=dyn_pouch2, bone2Name=dyn_pouch1, swing1LimitDegrees=1, swing2LimitDegrees=10, twistLimitDegrees=1}
 {bone1Name=dyn_pouch3, bone2Name=dyn_pouch2, swing1LimitDegrees=1, swing2LimitDegrees=20, twistLimitDegrees=1}
 {bone1Name=dyn_pouch4, bone2Name=dyn_pouch3, swing1LimitDegrees=1, swing2LimitDegrees=20, twistLimitDegrees=1}
 {bone1Name=dyn_pouch5, bone2Name=dyn_pouch4, swing1LimitDegrees=1, swing2LimitDegrees=20, twistLimitDegrees=1}
]
