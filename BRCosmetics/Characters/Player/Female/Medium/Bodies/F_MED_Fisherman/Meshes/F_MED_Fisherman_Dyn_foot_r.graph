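# Physics asset for the right foot of the F_MED_Fisherman female character body, set up for dynamic motion of the shoe tongue bone.
# Sized capsules sit on the calf and shoe tongue, while the foot and tongue master bones carry zero-size capsules as chain links.
# Constraints chain tongue, tongue master, foot and calf, with angular limits on the tongue joint and free swing and twist on the other two.
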PhysicsAsset /BRCosmetics/Characters/Player/Female/Medium/Bodies/F_MED_Fisherman/Meshes/F_MED_Fisherman_Dyn_foot_r
bodies {
 calf_r [{shapeType=Capsule, length=13.8535, radius=3.2373, center=<33.2747 3.5659 1.32>, rotation=<rot 90 -1.7899 3.2102>}]
 dyn_shoe_tongue_r [{shapeType=Capsule, length=2, radius=2, center=<1 0 0>, rotation=<rot 90 0 0>}]
 dyn_shoe_tongue_r_MSTR [{shapeType=Capsule, length=0, radius=0, shapeName=dyn_shoe_tongue_r_MSTR_capsule}]
 foot_r [{shapeType=Capsule, length=0, radius=0, shapeName=foot_r_capsule}]
}
constraints [
 {bone1Name=dyn_shoe_tongue_r, bone2Name=dyn_shoe_tongue_r_MSTR, swing1LimitDegrees=20, swing2LimitDegrees=90, twistLimitDegrees=5}
 {
  bone1Name=dyn_shoe_tongue_r_MSTR
  bone2Name=foot_r
  swing1LimitDegrees=45
  swing1Motion=Free
  swing2LimitDegrees=45
  swing2Motion=Free
  twistLimitDegrees=45
  twistMotion=Free
 }
 {
  bone1Name=foot_r
  bone2Name=calf_r
  swing1LimitDegrees=45
  swing1Motion=Free
  swing2LimitDegrees=45
  swing2Motion=Free
  twistLimitDegrees=45
  twistMotion=Free
 }
]
