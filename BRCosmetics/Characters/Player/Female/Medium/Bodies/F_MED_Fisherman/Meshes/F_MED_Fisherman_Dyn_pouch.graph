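# Physics asset for the dynamic pouch on the female medium Fisherman body cosmetic, defining collision bodies and joint limits for its bone chain.
# A tiny capsule on dyn_pouch_MSTR roots the chain, followed by a large pouch box and five generally smaller boxes, each constrained to the previous bone.
# The root joint swings widest, while later joints start tight and loosen toward the tip, with twist kept narrow throughout.
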PhysicsAsset /BRCosmetics/Characters/Player/Female/Medium/Bodies/F_MED_Fisherman/Meshes/F_MED_Fisherman_Dyn_pouch
bodies {
 dyn_pouch [{shapeType=Box, extentX=12, extentY=11.125, extentZ=5.1875, center=<1.7303 0.4037 2.6973>, rotation=<rot 7.5192 8.275 -0.2262>}]
 dyn_pouch1 [{shapeType=Box, extentX=5.875, extentY=7.5, extentZ=1.4375, center=<2.7912 0.0001 0>}]
 dyn_pouch2 [{shapeType=Box, extentX=4.375, extentY=7.5, extentZ=1.0625, center=<2.3785 0.0001 0>}]
 dyn_pouch3 [{shapeType=Box, extentX=4.375, extentY=7.5, extentZ=1.0625, center=<2.26 0.0001 0.0165>, rotation=<rot 2.8962 0 0>}]
 dyn_pouch4 [{shapeType=Box, extentX=2.9375, extentY=1.6875, extentZ=0.625, center=<1.6381 0.0001 -0.3841>, rotation=<rot -22.0386 0 0>}]
 dyn_pouch5 [{shapeType=Box, extentX=2.25, extentY=1.6875, extentZ=0.625, center=<1.0513 0.0001 0.1218>, rotation=<rot 5.8863 0 0>}]
 dyn_pouch_MSTR [{shapeType=Capsule, length=0.1, radius=0.1}]
}
constraints [
 {bone1Name=dyn_pouch, bone2Name=dyn_pouch_MSTR, swing1LimitDegrees=45, swing2LimitDegrees=45, twistLimitDegrees=5}
 {bone1Name=dyn_pouch1, bone2Name=dyn_pouch, swing1LimitDegrees=5, swing2LimitDegrees=10, twistLimitDegrees=5}
 {bone1Name=dyn_pouch2, bone2Name=dyn_pouch1, swing1LimitDegrees=5, swing2LimitDegrees=10, twistLimitDegrees=5}
 {bone1Name=dyn_pouch3, bone2Name=dyn_pouch2, swing1LimitDegrees=5, swing2LimitDegrees=20, twistLimitDegrees=5}
 {bone1Name=dyn_pouch4, bone2Name=dyn_pouch3, swing1LimitDegrees=15, swing2LimitDegrees=20, twistLimitDegrees=5}
 {bone1Name=dyn_pouch5, bone2Name=dyn_pouch4, swing1LimitDegrees=15, swing2LimitDegrees=20, twistLimitDegrees=5}
]
