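# Post-copy animation blueprint for the female medium Fisherman cosmetic body, adding corrective bone adjustments and secondary physics to the pose copied from its attached parent mesh.
# Crouch events drive an interpolated offset of the talkie bone.
# Per-frame thigh yaw relative to the pelvis drives additive corrections on left and right pelvis bones.
# Chained RigidBody simulations at both feet and pelvis finish the pose; a pouch simulation node sits unconnected.
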AnimBlueprint /BRCosmetics/Characters/Player/Female/Medium/Bodies/F_MED_Fisherman/Meshes/F_MED_Fisherman_AnimBP
target_skeleton=&/F_MED_Fisherman_Skeleton
graphs [
 AnimGraph {nodes [
   RigidBody() {
    bClampLinearTranslationLimitToRefPose=true
    bOverrideWorldGravity=true
    overridePhysicsAsset=&/F_MED_Fisherman_Dyn_pouch
    simulationSpace=BaseBoneSpace
    alphaBoolBlend {blendOption=Linear}
    baseBoneRef {boneName=thigh_r}
    componentAppliedLinearAccClamp=<3000 3000 3000>
    componentLinearAccScale=<0.3 0.3 0.3>
    componentLinearVelScale=<0.3 0.3 0.3>
    overrideWorldGravity {z=-980}
   }
   n1 = "Transform (Modify) Bone - Bone: dyn_talkie_MSTR"(
    ComponentPose=Local To Component(LocalPose=Copy Pose From Mesh() {bUseAttachedParent=true})
    Alpha=Get crouch alpha() -> (crouch alpha:double)
    Translation="0.000,-3.572,-0.505"
    Rotation="0.000,-0.000,-27.217"
    Scale="1,1,1"
   ) {
    rotationMode=BMM_Additive
    rotationSpace=BCS_BoneSpace
    translationMode=BMM_Additive
    translationSpace=BCS_BoneSpace
    alphaBoolBlend {blendOption=Linear}
    alphaScaleBiasClamp {bInterpResult=true, interpSpeedDecreasing=20, interpSpeedIncreasing=20}
    boneToModify {boneName=dyn_talkie_MSTR}
    rotation {roll=-27.217}
    translation {y=-3.572, z=-0.505}
   }
   n2 = Get thigh_rot_l()
   n3 = "Transform (Modify) Bone - Bone: pelvis_l"(ComponentPose=@n1, Alpha=@n2, Translation="-1.599,0.094,0.175", Rotation="0,0,0", Scale="1.207,1.000,1.000") {
    scaleMode=BMM_Additive
    translationMode=BMM_Additive
    translationSpace=BCS_ParentBoneSpace
    alphaBoolBlend {blendOption=Linear}
    boneToModify {boneName=pelvis_l}
    scale {x=1.2069}
    translation=<-1.599 0.094 0.175>
   }
   n4 = "Transform (Modify) Bone - Bone: pelvis_mid_l"(ComponentPose=@n3, Alpha=@n2, Translation="-2.366,0.140,0.263", Rotation="0,0,0", Scale="1.216,1.000,1.000") {
    scaleMode=BMM_Additive
    translationMode=BMM_Additive
    translationSpace=BCS_ParentBoneSpace
    alphaBoolBlend {blendOption=Linear}
    boneToModify {boneName=pelvis_mid_l}
    scale {x=1.2163}
    translation=<-2.3657 0.14 0.263>
   }
   n5 = "Transform (Modify) Bone - Bone: pelvis_side_l"(ComponentPose=@n4, Alpha=@n2, Translation="0,0,0", Rotation="0,0,0", Scale="1,1,1") {
    scaleMode=BMM_Additive
    translationMode=BMM_Additive
    translationSpace=BCS_ParentBoneSpace
    alphaBoolBlend {blendOption=Linear}
    boneToModify {boneName=pelvis_side_l}
   }
   n6 = Get thigh_rot_r()
   n7 = "Transform (Modify) Bone - Bone: pelvis_r"(ComponentPose=@n5, Alpha=@n6, Translation="-1.599,0.094,0.175", Rotation="0,0,0", Scale="1.2,1,1") {
    scaleMode=BMM_Additive
    translationMode=BMM_Additive
    translationSpace=BCS_ParentBoneSpace
    alphaBoolBlend {blendOption=Linear}
    boneToModify {boneName=pelvis_r}
    scale {x=1.2}
    translation=<-1.599 0.094 0.175>
   }
   n8 = "Transform (Modify) Bone - Bone: pelvis_mid_r"(ComponentPose=@n7, Alpha=@n6, Translation="-0.564,0.032,0.062", Rotation="0,0,0", Scale="1.204,1.000,1.000") {
    scaleMode=BMM_Additive
    translationMode=BMM_Additive
    translationSpace=BCS_ParentBoneSpace
    alphaBoolBlend {blendOption=Linear}
    boneToModify {boneName=pelvis_mid_r}
    scale {x=1.2036}
    translation=<-0.5637 0.032 0.062>
   }
   n9 = "Transform (Modify) Bone - Bone: pelvis_side_r"(ComponentPose=@n8, Alpha=@n6, Translation="0,0,0", Rotation="0,0,0", Scale="1,1,1") {
    scaleMode=BMM_Additive
    translationMode=BMM_Additive
    translationSpace=BCS_ParentBoneSpace
    alphaBoolBlend {blendOption=Linear}
    boneToModify {boneName=pelvis_side_r}
   }
   n10 = RigidBody(ComponentPose=@n9) {
    bClampLinearTranslationLimitToRefPose=true
    bOverrideWorldGravity=true
    overridePhysicsAsset=&/F_MED_Fisherman_Dyn_foot_l
    simulationSpace=BaseBoneSpace
    alphaBoolBlend {blendOption=Linear}
    baseBoneRef {boneName=foot_l}
    componentAppliedLinearAccClamp=<3000 3000 3000>
    componentLinearAccScale=<0.6 0.5 0.5>
    componentLinearVelScale=<0.6 0.5 0.5>
    overrideWorldGravity {z=-980}
   }
   n11 = RigidBody(ComponentPose=@n10) {
    bClampLinearTranslationLimitToRefPose=true
    bOverrideWorldGravity=true
    overridePhysicsAsset=&/F_MED_Fisherman_Dyn_foot_r
    simulationSpace=BaseBoneSpace
    alphaBoolBlend {blendOption=Linear}
    baseBoneRef {boneName=foot_r}
    componentAppliedLinearAccClamp=<3000 3000 3000>
    componentLinearAccScale=<0.6 0.5 0.5>
    componentLinearVelScale=<0.6 0.5 0.5>
    overrideWorldGravity {z=-980}
   }
   n12 = RigidBody(ComponentPose=@n11) {
    bClampLinearTranslationLimitToRefPose=true
    bOverrideWorldGravity=true
    overridePhysicsAsset=&/F_MED_Fisherman_Dyn
    simulationSpace=BaseBoneSpace
    alphaBoolBlend {blendOption=Linear}
    baseBoneRef {boneName=pelvis}
    componentAppliedLinearAccClamp=<3000 3000 3000>
    componentLinearAccScale=<0.5 0.4 0.4>
    componentLinearVelScale=<0.5 0.4 0.4>
    overrideWorldGravity {z=-980}
   }
   Output Pose(Result=Component To Local(ComponentPose=@n12)) {layerGroup=None}
  ]}
 EventGraph {nodes [
   Event BlueprintUpdateAnimation()
   n1 = BreakTransform(InTransform=GetDeltaTransformFromRefPose(self=GetOwnerSkeletalMeshComponent(), BoneName=thigh_r, BaseName=pelvis))
   > Set thigh_rot_r(thigh_rot_r=MapRangeClamped(Value=BreakRotator(InRot=@n1.Rotation).Yaw, InRangeA=0, InRangeB=-90))
   n2 = BreakTransform(InTransform=GetDeltaTransformFromRefPose(self=GetOwnerSkeletalMeshComponent(), BoneName=thigh_l, BaseName=pelvis))
   > Set thigh_rot_l(thigh_rot_l=MapRangeClamped(Value=BreakRotator(InRot=@n2.Rotation).Yaw, InRangeA=-15, InRangeB=-110))
   Event OnBeginCrouching()
   > Set crouch alpha(crouch alpha:double=1) -> (Output_Get:double)
   Event OnEndCrouching()
   > Set crouch alpha(crouch alpha=0)
  ]}
]
variables [
 crouch alpha {type=number}
 thigh_rot_l {type=number}
 thigh_rot_r {type=number}
]
