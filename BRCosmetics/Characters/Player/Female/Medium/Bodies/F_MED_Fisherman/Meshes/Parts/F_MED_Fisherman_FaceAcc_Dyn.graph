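# Physics asset for the female medium Fisherman's FaceAcc part, defining dynamic hair and braid physics.
# Capsule chains for two left braids, a front-right strand and three back-hair chains hang from the head with limited swing and twist.
# Spine, clavicle, neck and head capsules form the upper body, joined by angularly free constraints.
# Similar free constraints cross-link both side back-hair tips to the center chain's third bone, tying the back hair together.
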PhysicsAsset /BRCosmetics/Characters/Player/Female/Medium/Bodies/F_MED_Fisherman/Meshes/Parts/F_MED_Fisherman_FaceAcc_Dyn
bodies {
 clavicle_l [{shapeType=Capsule, length=9.4739, radius=5.9636, center=<6.9438 -0.8931 -1.0776>, rotation=<rot 62.9355 0 0>}]
 clavicle_r [{shapeType=Capsule, length=16.0651, radius=5.8539, center=<-9.4543 0.8929 1.7899>, rotation=<rot 67.8041 0 0>}]
 dyn_braid_bk_l_01 [{shapeType=Capsule, length=3, radius=1, center=<2 0 0>, rotation=<rot 90 0 0>}]
 dyn_braid_bk_l_02 [{shapeType=Capsule, length=3, radius=1, center=<2 0 0>, rotation=<rot 90 0 0>}]
 dyn_braid_bk_l_03 [{shapeType=Capsule, length=3, radius=1, center=<2 0 0>, rotation=<rot 90 0 0>}]
 dyn_braid_bk_l_04 [{shapeType=Capsule, length=3, radius=1, center=<2 0 0>, rotation=<rot 90 0 0>}]
 dyn_braid_bk_l_05 [{shapeType=Capsule, length=3, radius=1, center=<2 0 0>, rotation=<rot 90 0 0>}]
 dyn_braid_fr_l_01 [{shapeType=Capsule, length=3, radius=1, center=<2 0 0>, rotation=<rot 90 0 0>}]
 dyn_braid_fr_l_02 [{shapeType=Capsule, length=3, radius=1, center=<2 0 0>, rotation=<rot 90 0 0>}]
 dyn_braid_fr_l_03 [{shapeType=Capsule, length=3, radius=1, center=<2 0 0>, rotation=<rot 90 0 0>}]
 dyn_braid_fr_l_04 [{shapeType=Capsule, length=3, radius=1, center=<2 0 0>, rotation=<rot 90 0 0>}]
 dyn_braid_fr_l_05 [{shapeType=Capsule, length=3, radius=1, center=<2 0 0>, rotation=<rot 90 0 0>}]
 dyn_hair_bk_01 [{shapeType=Capsule, length=4, radius=2.75, center=<2.7 0 0.0001>, rotation=<rot 0 0 90>}]
 dyn_hair_bk_02 [{shapeType=Capsule, length=4, radius=2.75, center=<2.7 0 0.0001>, rotation=<rot 0 0 90>}]
 dyn_hair_bk_03 [{shapeType=Capsule, length=4, radius=2.75, center=<2.7 0 0.0001>, rotation=<rot 0 0 90>}]
 dyn_hair_bk_04 [{shapeType=Capsule, length=4, radius=2.75, center=<2.7 0 0.0001>, rotation=<rot 0 0 90>}]
 dyn_hair_bk_l_01 [{shapeType=Capsule, length=4, radius=2.75, center=<2.7 0 0.0001>, rotation=<rot 0 0 90>}]
 dyn_hair_bk_l_02 [{shapeType=Capsule, length=4, radius=2.75, center=<2.7 0 0.0001>, rotation=<rot 0 0 90>}]
 dyn_hair_bk_l_03 [{shapeType=Capsule, length=4, radius=2.75, center=<2.7 0 0.0001>, rotation=<rot 0 0 90>}]
 dyn_hair_bk_r_01 [{shapeType=Capsule, length=4, radius=2.75, center=<2.7 0 0.0001>, rotation=<rot 0 0 90>}]
 dyn_hair_bk_r_02 [{shapeType=Capsule, length=4, radius=2.75, center=<2.7 0 0.0001>, rotation=<rot 0 0 90>}]
 dyn_hair_bk_r_03 [
  {shapeType=Capsule, length=4, radius=2.75, center=<2.7 0 0.0001>, rotation=<rot 0 0 90>}
  {shapeType=Capsule, length=4, radius=2.75, center=<10.8881 0 -0.0003>, rotation=<rot 0 0 90.0001>}
 ]
 dyn_hair_fr_r_01 [{shapeType=Capsule, length=1.1875, radius=1, center=<1.1609 0 0>, rotation=<rot 90 0 0>}]
 dyn_hair_fr_r_02 [{shapeType=Capsule, length=1.5938, radius=1, center=<1.195 0 0>, rotation=<rot 90 0 0>}]
 head [{shapeType=Capsule, length=7.323, radius=6.2576, center=<3.0452 0.7617 0.0003>, rotation=<rot 90 0 0>}]
 neck_01 [{shapeType=Capsule, length=3.1738, radius=4.3756, center=<-0.0588 -0.7719 0>, rotation=<rot 90 0 0>}]
 neck_02 [{shapeType=Capsule, length=2.5418, radius=5.0076, center=<0 -0.5834 0>, rotation=<rot 90 0 0>}]
 spine_04 [{shapeType=Capsule, length=10.4793, radius=10.0701, center=<2.3544 1.8797 0>}]
 spine_05 [{shapeType=Capsule, length=13.5511, radius=6.9983, center=<5.3328 -1.6287 0.0004>}]
}
constraints [
 {bone1Name=dyn_hair_fr_r_02, bone2Name=dyn_hair_fr_r_01, swing1LimitDegrees=20, swing2LimitDegrees=20, twistLimitDegrees=5}
 {bone1Name=dyn_braid_bk_l_02, bone2Name=dyn_braid_bk_l_01, swing1LimitDegrees=20, swing2LimitDegrees=20, twistLimitDegrees=5}
 {bone1Name=dyn_braid_bk_l_03, bone2Name=dyn_braid_bk_l_02, swing1LimitDegrees=45, swing2LimitDegrees=45, twistLimitDegrees=5}
 {bone1Name=dyn_braid_bk_l_04, bone2Name=dyn_braid_bk_l_03, swing1LimitDegrees=45, swing2LimitDegrees=45, twistLimitDegrees=5}
 {bone1Name=dyn_braid_bk_l_05, bone2Name=dyn_braid_bk_l_04, swing1LimitDegrees=45, swing2LimitDegrees=45, twistLimitDegrees=5}
 {bone1Name=dyn_braid_fr_l_02, bone2Name=dyn_braid_fr_l_01, swing1LimitDegrees=20, swing2LimitDegrees=20, twistLimitDegrees=5}
 {bone1Name=dyn_braid_fr_l_03, bone2Name=dyn_braid_fr_l_02, swing1LimitDegrees=45, swing2LimitDegrees=45, twistLimitDegrees=5}
 {bone1Name=dyn_braid_fr_l_04, bone2Name=dyn_braid_fr_l_03, swing1LimitDegrees=45, swing2LimitDegrees=45, twistLimitDegrees=5}
 {bone1Name=dyn_braid_fr_l_05, bone2Name=dyn_braid_fr_l_04, swing1LimitDegrees=45, swing2LimitDegrees=45, twistLimitDegrees=5}
 {bone1Name=dyn_hair_bk_r_02, bone2Name=dyn_hair_bk_r_01, swing1LimitDegrees=20, swing2LimitDegrees=60, twistLimitDegrees=20}
 {bone1Name=dyn_hair_bk_r_03, bone2Name=dyn_hair_bk_r_02, swing1LimitDegrees=20, swing2LimitDegrees=60, twistLimitDegrees=20}
 {bone1Name=dyn_hair_bk_l_02, bone2Name=dyn_hair_bk_l_01, swing1LimitDegrees=20, swing2LimitDegrees=60, twistLimitDegrees=15}
 {bone1Name=dyn_hair_bk_l_03, bone2Name=dyn_hair_bk_l_02, swing1LimitDegrees=20, swing2LimitDegrees=60, twistLimitDegrees=15}
 {bone1Name=dyn_hair_bk_02, bone2Name=dyn_hair_bk_01, swing1LimitDegrees=30, swing2LimitDegrees=45, twistLimitDegrees=5}
 {bone1Name=dyn_hair_bk_03, bone2Name=dyn_hair_bk_02, swing1LimitDegrees=30, swing2LimitDegrees=45, twistLimitDegrees=5}
 {bone1Name=dyn_hair_bk_04, bone2Name=dyn_hair_bk_03, swing1LimitDegrees=30, swing2LimitDegrees=45, twistLimitDegrees=5}
 {bone1Name=dyn_hair_fr_r_01, bone2Name=head, swing1LimitDegrees=20, swing2LimitDegrees=20, twistLimitDegrees=5}
 {bone1Name=dyn_braid_bk_l_01, bone2Name=head, swing1LimitDegrees=20, swing2LimitDegrees=20, twistLimitDegrees=5}
 {bone1Name=dyn_braid_fr_l_01, bone2Name=head, swing1LimitDegrees=20, swing2LimitDegrees=20, twistLimitDegrees=5}
 {bone1Name=dyn_hair_bk_r_01, bone2Name=head, swing1LimitDegrees=20, swing2LimitDegrees=60, twistLimitDegrees=15}
 {bone1Name=dyn_hair_bk_l_01, bone2Name=head, swing1LimitDegrees=20, swing2LimitDegrees=60, twistLimitDegrees=15}
 {bone1Name=dyn_hair_bk_01, bone2Name=head, swing1LimitDegrees=30, swing2LimitDegrees=45, twistLimitDegrees=5}
 {
  bone1Name=clavicle_l
  bone2Name=spine_05
  swing1LimitDegrees=45
  swing1Motion=Free
  swing2LimitDegrees=45
  swing2Motion=Free
  twistLimitDegrees=45
  twistMotion=Free
 }
 {
  bone1Name=clavicle_r
  bone2Name=spine_05
  swing1LimitDegrees=45
  swing1Motion=Free
  swing2LimitDegrees=45
  swing2Motion=Free
  twistLimitDegrees=45
  twistMotion=Free
 }
 {
  bone1Name=head
  bone2Name=neck_02
  swing1LimitDegrees=45
  swing1Motion=Free
  swing2LimitDegrees=45
  swing2Motion=Free
  twistLimitDegrees=45
  twistMotion=Free
 }
 {
  bone1Name=neck_01
  bone2Name=spine_05
  swing1LimitDegrees=45
  swing1Motion=Free
  swing2LimitDegrees=45
  swing2Motion=Free
  twistLimitDegrees=45
  twistMotion=Free
 }
 {
  bone1Name=neck_02
  bone2Name=neck_01
  swing1LimitDegrees=45
  swing1Motion=Free
  swing2LimitDegrees=45
  swing2Motion=Free
  twistLimitDegrees=45
  twistMotion=Free
 }
 {
  bone1Name=spine_05
  bone2Name=spine_04
  swing1LimitDegrees=45
  swing1Motion=Free
  swing2LimitDegrees=45
  swing2Motion=Free
  twistLimitDegrees=45
  twistMotion=Free
 }
 {
  bone1Name=dyn_hair_bk_r_03
  bone2Name=dyn_hair_bk_03
  swing1LimitDegrees=45
  swing1Motion=Free
  swing2LimitDegrees=45
  swing2Motion=Free
  twistLimitDegrees=45
  twistMotion=Free
 }
 {
  bone1Name=dyn_hair_bk_l_03
  bone2Name=dyn_hair_bk_03
  swing1LimitDegrees=45
  swing1Motion=Free
  swing2LimitDegrees=45
  swing2Motion=Free
  twistLimitDegrees=45
  twistMotion=Free
 }
]
